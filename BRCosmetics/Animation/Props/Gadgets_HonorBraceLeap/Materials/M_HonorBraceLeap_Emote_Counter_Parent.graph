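# Parent material for the HonorBraceLeap emote prop's counter: an unlit, masked six-digit rolling odometer over a background texture, tinted by time of day.
# DynamicParameter Time and GlobalSpeed drive it, with each higher digit dividing both by ten.
# Digits scroll a number-strip texture with a power-curve snap, slanted UVs and crossfades between its channels to fake speed-based motion blur.
# At the 99999.9 cap the value clamps and the digits pulse on a sine.
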Material /BRCosmetics/Animation/Props/Gadgets_HonorBraceLeap/Materials/M_HonorBraceLeap_Emote_Counter_Parent
graph {
 properties {bAutomaticallySetUsageInEditor=false, bHasPixelAnimation=true, blendMode=BLEND_Masked, shadingModel=MSM_Unlit, usedShadingModels=Unlit}
 nodes [
  n1 = NamedRerouteUsage()
  n2 = NamedRerouteUsage()
  n3 = Constant3Vector() {constant=<rgba 1 1 1 0>}
  Multiply(
   A=Add(
    A=TextureSample() {texture=/BRCosmetics/Animation/Props/Gadgets_HonorBraceLeap/Textures/T_HonorBraceLeap_Emote_CounterBackground}.RGB
    B=Multiply(A=@n1, B=@n2)
   )
   B=MaterialFunctionCall(Color_Morning=@n3, Color_Day=@n3, Color_Evening=@n3, Color_Night=Constant3Vector() {constant=<rgba 0.5 0.5833 1 0>}) {materialFunction=&Color_Time_of_Day}.Result
  )
  n4 = DynamicParameter() {defaultValue=<rgba 0 0 1 1>, paramNames [Time,GlobalSpeed,Param3,Param4]}
  NamedRerouteDeclaration(None=@n4.GlobalSpeed) {name=GlobalSpeed, nodeColor=<rgba 1 0 0.8906 1>}
  n5 = Divide(A=@n4.Time, B=NamedRerouteUsage())
  NamedRerouteDeclaration(None=Saturate(None=Floor(None=@n5))) {name=MaxNumReached, nodeColor=<rgba 1 0 0.9609 1>}
  n6 = TextureCoordinate()
  n7 = ComponentMask(None=@n6) {g=true}
  n8 = MaterialFunctionCall(Input=@n7, Input Low=Constant() {r=0.45}, Input High=Constant() {r=0.55}) {materialFunction=&RemapValueRange}
  n9 = Constant() {r=-0.04}
  n10 = Add(
   A=ComponentMask(None=@n6) {r=true}
   B=Multiply(A=Multiply(A=@n8.Result), B=NamedRerouteDeclaration(None=@n9) {name=SlantStrength, nodeColor=<rgba 0 1 0.0859 1>})
  )
  n11 = Constant() {r=0.2}
  n12 = Multiply(
   A=MaterialFunctionCall(UVs=AppendVector(A=@n10, B=@n7), Texture Scale=NamedRerouteDeclaration(None=@n11) {name=DigitScale, nodeColor=<rgba 0 0.9531 1 1>}) {materialFunction=&ScaleUVsByCenter}.UVs
   B=Constant2Vector() {g=0.1111, r=1}
  )
  n13 = Constant2Vector() {r=-1.65}
  n14 = NamedRerouteDeclaration(None=Add(
    A=NamedRerouteDeclaration(None=@n12) {name=Digit_UVs, nodeColor=<rgba 0.4062 0 1 1>}
    B=NamedRerouteDeclaration(None=@n13) {name=RowStart, nodeColor=<rgba 0.3125 1 0 1>}
   )) {name=UVs_0, nodeColor=<rgba 0.7344 0 1 1>}
  n15 = Saturate(None=ComponentMask(None=@n14) {r=true})
  n16 = ComponentMask(None=@n14) {g=true}
  n17 = NamedRerouteDeclaration(None=Constant() {r=99999.8984}) {name=MaxNum, nodeColor=<rgba 1 0 0.4219 1>}
  n18 = NamedRerouteUsage()
  n19 = Frac(None=NamedRerouteDeclaration(None=LinearInterpolate(A=@n4.Time, B=@n17, Alpha=@n18) {constB=99999.8984}) {name=Time, nodeColor=<rgba 1 0.3984 0 1>})
  n20 = NamedRerouteUsage()
  n21 = Multiply(A=Multiply(A=LinearInterpolate(A=@n19, Alpha=@n20) {constB=0.9}, B=Constant2Vector() {g=1})) {constB=10}
  n22 = Frac(None=@n21)
  n23 = Divide(A=Constant() {r=1}, B=NamedRerouteUsage())
  n24 = Add(
   A=Subtract(A=@n21, B=@n22)
   B=Power(Base=@n22, Exp=NamedRerouteDeclaration(None=@n23) {name=Power, nodeColor=<rgba 1 0.7266 0 1>}) {constExponent=5}
  )
  n25 = TextureSample(UVs=Add(A=AppendVector(A=@n15, B=@n16), B=Divide(A=@n24) {constB=10})) {texture=/BRCosmetics/Animation/Props/Gadgets_HonorBraceLeap/Textures/T_HonorBraceLeap_Emote_BlurredNumbers}
  n26 = NamedRerouteDeclaration(None=LinearInterpolate(A=NamedRerouteUsage(), Alpha=NamedRerouteUsage()) {constB=0}) {name=Speed_0, nodeColor=<rgba 0.9453 1 0 1>}
  n27 = NamedRerouteDeclaration(None=Constant() {r=0.1}) {name=MinSpeed, nodeColor=<rgba 0.6406 1 0 1>}
  n28 = Constant() {r=0.5}
  n29 = Divide(A=Subtract(A=@n26, B=@n27) {constB=0.1}, B=Subtract(A=@n28, B=@n27)) {constB=0.15}
  n30 = NamedRerouteDeclaration(None=@n28) {name=MidSpeed, nodeColor=<rgba 1 0 0.9609 1>}
  n31 = Subtract(A=@n26, B=@n30) {constB=0.25}
  n32 = Subtract(A=NamedRerouteDeclaration(None=Constant() {r=10}) {name=MaxSpeed, nodeColor=<rgba 0 1 0.7188 1>}, B=@n30) {constB=0.25}
  n33 = LinearInterpolate(A=LinearInterpolate(A=@n25.R, B=@n25.R, Alpha=Saturate(None=@n29)), B=@n25.B, Alpha=Saturate(None=Divide(A=@n31, B=@n32) {constB=0.75}))
  n34 = NamedRerouteUsage()
  n35 = NamedRerouteDeclaration(None=Constant2Vector() {r=0.5}) {name=RowOffset, nodeColor=<rgba 1 0 0.6328 1>}
  n36 = NamedRerouteDeclaration(None=Add(A=NamedRerouteUsage(), B=Add(A=@n34, B=@n35))) {name=UVs_1, nodeColor=<rgba 0.7344 0 1 1>}
  n37 = Saturate(None=ComponentMask(None=@n36) {r=true})
  n38 = ComponentMask(None=@n36) {g=true}
  n39 = NamedRerouteUsage()
  n40 = Multiply(A=Multiply(A=Divide(A=@n39) {constB=10}, B=Constant2Vector() {g=1})) {constB=10}
  n41 = Frac(None=@n40)
  n42 = NamedRerouteUsage()
  n43 = Add(A=Subtract(A=@n40, B=@n41), B=Power(Base=@n41, Exp=Multiply(A=@n42) {constB=10}) {constExponent=5})
  n44 = TextureSample(UVs=Add(A=AppendVector(A=@n37, B=@n38), B=Divide(A=@n43) {constB=10})) {texture=&T_HonorBraceLeap_Emote_BlurredNumbers}
  n45 = NamedRerouteDeclaration(None=Divide(A=NamedRerouteUsage()) {constB=10}) {name=Speed_1, nodeColor=<rgba 0.9453 1 0 1>}
  n46 = NamedRerouteUsage()
  n47 = NamedRerouteUsage()
  n48 = Divide(A=Subtract(A=@n45, B=@n46) {constB=0.1}, B=Subtract(A=@n47, B=@n46)) {constB=0.15}
  n49 = Subtract(A=@n45, B=@n47) {constB=0.25}
  n50 = Subtract(A=NamedRerouteUsage(), B=@n47)
  n51 = LinearInterpolate(A=LinearInterpolate(A=@n44.R, B=@n44.R, Alpha=Saturate(None=@n48)), B=@n44.B, Alpha=Saturate(None=Divide(A=@n49, B=@n50) {constB=0.75}))
  n52 = NamedRerouteUsage()
  n53 = Multiply(A=NamedRerouteUsage()) {constB=2}
  n54 = NamedRerouteDeclaration(None=Add(A=NamedRerouteUsage(), B=Add(A=@n52, B=@n53))) {name=UVs_2, nodeColor=<rgba 0.7344 0 1 1>}
  n55 = Saturate(None=ComponentMask(None=@n54) {r=true})
  n56 = ComponentMask(None=@n54) {g=true}
  n57 = NamedRerouteUsage()
  n58 = Multiply(A=Multiply(A=Divide(A=@n57) {constB=100}, B=Constant2Vector() {g=1})) {constB=10}
  n59 = Frac(None=@n58)
  n60 = NamedRerouteUsage()
  n61 = Add(A=Subtract(A=@n58, B=@n59), B=Power(Base=@n59, Exp=Multiply(A=@n60) {constB=100}) {constExponent=5})
  n62 = TextureSample(UVs=Add(A=AppendVector(A=@n55, B=@n56), B=Divide(A=@n61) {constB=10})) {texture=&T_HonorBraceLeap_Emote_BlurredNumbers}
  n63 = NamedRerouteDeclaration(None=Divide(A=NamedRerouteUsage()) {constB=10}) {name=Speed_2, nodeColor=<rgba 0.9453 1 0 1>}
  n64 = NamedRerouteUsage()
  n65 = NamedRerouteUsage()
  n66 = Divide(A=Subtract(A=@n63, B=@n64) {constB=0.1}, B=Subtract(A=@n65, B=@n64)) {constB=0.15}
  n67 = Subtract(A=@n63, B=@n65) {constB=0.25}
  n68 = Subtract(A=NamedRerouteUsage(), B=@n65)
  n69 = LinearInterpolate(A=LinearInterpolate(A=@n62.R, B=@n62.R, Alpha=Saturate(None=@n66)), B=@n62.B, Alpha=Saturate(None=Divide(A=@n67, B=@n68) {constB=0.75}))
  n70 = NamedRerouteUsage()
  n71 = Multiply(A=NamedRerouteUsage()) {constB=3}
  n72 = NamedRerouteDeclaration(None=Add(A=NamedRerouteUsage(), B=Add(A=@n70, B=@n71))) {name=UVs_3, nodeColor=<rgba 0.7344 0 1 1>}
  n73 = Saturate(None=ComponentMask(None=@n72) {r=true})
  n74 = ComponentMask(None=@n72) {g=true}
  n75 = NamedRerouteUsage()
  n76 = Multiply(A=Multiply(A=Divide(A=@n75) {constB=1000}, B=Constant2Vector() {g=1})) {constB=10}
  n77 = Frac(None=@n76)
  n78 = NamedRerouteUsage()
  n79 = Add(A=Subtract(A=@n76, B=@n77), B=Power(Base=@n77, Exp=Multiply(A=@n78) {constB=1000}) {constExponent=5})
  n80 = TextureSample(UVs=Add(A=AppendVector(A=@n73, B=@n74), B=Divide(A=@n79) {constB=10})) {texture=&T_HonorBraceLeap_Emote_BlurredNumbers}
  n81 = NamedRerouteDeclaration(None=Divide(A=NamedRerouteUsage()) {constB=10}) {name=Speed_3, nodeColor=<rgba 0.9453 1 0 1>}
  n82 = NamedRerouteUsage()
  n83 = NamedRerouteUsage()
  n84 = Divide(A=Subtract(A=@n81, B=@n82) {constB=0.1}, B=Subtract(A=@n83, B=@n82)) {constB=0.15}
  n85 = Subtract(A=@n81, B=@n83) {constB=0.25}
  n86 = Subtract(A=NamedRerouteUsage(), B=@n83)
  n87 = LinearInterpolate(A=LinearInterpolate(A=@n80.R, B=@n80.R, Alpha=Saturate(None=@n84)), B=@n80.B, Alpha=Saturate(None=Divide(A=@n85, B=@n86) {constB=0.75}))
  n88 = Add(
   A=Add(
    A=NamedRerouteDeclaration(None=@n33) {name=Digit_0, nodeColor=<rgba 0 1 0.9766 1>}
    B=NamedRerouteDeclaration(None=@n51) {name=Digit_1, nodeColor=<rgba 0 1 0.9766 1>}
   )
   B=Add(
    A=NamedRerouteDeclaration(None=@n69) {name=Digit_2, nodeColor=<rgba 0 1 0.9766 1>}
    B=NamedRerouteDeclaration(None=@n87) {name=Digit_3, nodeColor=<rgba 0 1 0.9766 1>}
   )
  )
  n89 = NamedRerouteUsage()
  n90 = Multiply(A=NamedRerouteUsage()) {constB=4}
  n91 = NamedRerouteDeclaration(None=Add(A=NamedRerouteUsage(), B=Add(A=@n89, B=@n90))) {name=UVs_4, nodeColor=<rgba 0.7344 0 1 1>}
  n92 = Saturate(None=ComponentMask(None=@n91) {r=true})
  n93 = ComponentMask(None=@n91) {g=true}
  n94 = NamedRerouteUsage()
  n95 = Multiply(A=Multiply(A=Divide(A=@n94) {constB=10000}, B=Constant2Vector() {g=1})) {constB=10}
  n96 = Frac(None=@n95)
  n97 = NamedRerouteUsage()
  n98 = Add(A=Subtract(A=@n95, B=@n96), B=Power(Base=@n96, Exp=Multiply(A=@n97) {constB=10000}) {constExponent=5})
  n99 = TextureSample(UVs=Add(A=AppendVector(A=@n92, B=@n93), B=Divide(A=@n98) {constB=10})) {texture=&T_HonorBraceLeap_Emote_BlurredNumbers}
  n100 = NamedRerouteDeclaration(None=Divide(A=NamedRerouteUsage()) {constB=10}) {name=Speed_4, nodeColor=<rgba 0.9453 1 0 1>}
  n101 = NamedRerouteUsage()
  n102 = Subtract(A=@n100, B=@n101) {constB=0.1}
  n103 = NamedRerouteUsage()
  n104 = Subtract(A=@n103, B=@n101)
  n105 = LinearInterpolate(A=@n99.R, B=@n99.R, Alpha=Saturate(None=Divide(A=@n102, B=@n104) {constB=0.15}))
  n106 = NamedRerouteUsage()
  n107 = Saturate(None=Divide(A=Subtract(A=@n100, B=@n103) {constB=0.25}, B=Subtract(A=@n106, B=@n103)) {constB=0.75})
  n108 = NamedRerouteUsage()
  n109 = Multiply(A=NamedRerouteUsage()) {constB=5}
  n110 = NamedRerouteDeclaration(None=Add(A=NamedRerouteUsage(), B=Add(A=@n108, B=@n109))) {name=UVs_5, nodeColor=<rgba 0.7344 0 1 1>}
  n111 = Saturate(None=ComponentMask(None=@n110) {r=true})
  n112 = ComponentMask(None=@n110) {g=true}
  n113 = NamedRerouteUsage()
  n114 = Multiply(A=Multiply(A=Divide(A=@n113) {constB=100000}, B=Constant2Vector() {g=1})) {constB=10}
  n115 = Frac(None=@n114)
  n116 = NamedRerouteUsage()
  n117 = Add(A=Subtract(A=@n114, B=@n115), B=Power(Base=@n115, Exp=Multiply(A=@n116) {constB=100000}) {constExponent=5})
  n118 = TextureSample(UVs=Add(A=AppendVector(A=@n111, B=@n112), B=Divide(A=@n117) {constB=10})) {texture=&T_HonorBraceLeap_Emote_BlurredNumbers}
  n119 = NamedRerouteDeclaration(None=Divide(A=NamedRerouteUsage()) {constB=10}) {name=Speed_5, nodeColor=<rgba 0.9453 1 0 1>}
  n120 = NamedRerouteUsage()
  n121 = Saturate(None=Subtract(A=@n119, B=@n120) {constB=0.1})
  n122 = NamedRerouteUsage()
  n123 = Subtract(A=@n122, B=@n120)
  n124 = LinearInterpolate(A=@n118.R, B=@n118.R, Alpha=Saturate(None=Divide(A=@n121, B=@n123) {constB=0.15}))
  n125 = Subtract(A=@n119, B=@n122) {constB=0.25}
  n126 = NamedRerouteUsage()
  n127 = Saturate(None=Divide(A=Saturate(None=@n125), B=Subtract(A=@n126, B=@n122)) {constB=0.75})
  n128 = Add(
   A=NamedRerouteDeclaration(None=LinearInterpolate(A=@n105, B=@n99.B, Alpha=@n107)) {name=Digit_4, nodeColor=<rgba 0 1 0.9766 1>}
   B=NamedRerouteDeclaration(None=LinearInterpolate(A=@n124, B=@n118.B, Alpha=@n127)) {name=Digit_5, nodeColor=<rgba 0 1 0.9766 1>}
  )
  n129 = Time()
  n130 = ConstantBiasScale(None=Sine(None=Multiply(A=@n129) {constB=4}))
  n131 = NamedRerouteUsage()
  NamedRerouteDeclaration(None=Multiply(A=Add(A=@n88, B=@n128), B=LinearInterpolate(B=@n130, Alpha=@n131) {constA=1})) {name=Digits, nodeColor=<rgba 1 0 0.3047 1>}
  n132 = ComponentMask(None=TextureCoordinate()) {g=true}
  n133 = Abs(None=ConstantBiasScale(None=Add(A=@n132) {constB=0.02}) {bias=-0.5, scale=2})
  NamedRerouteDeclaration(None=SmoothStep(Value=OneMinus(None=@n133)) {constMin=0.78}) {name=DigitsMask, nodeColor=<rgba 0 0.625 1 1>}
 ]
}
